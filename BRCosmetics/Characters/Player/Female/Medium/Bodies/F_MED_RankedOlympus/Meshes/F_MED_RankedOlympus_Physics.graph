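# Physics asset for the female medium RankedOlympus cosmetic body, set up for dangling cloth bones with simplified lower-body collision.
# Two four-bone front chains and a three-bone back chain hang from the pelvis as capsules with tight angular limits.
# The lower two front segment pairs are cross-linked left to right with free rotation.
# Pelvis, thigh and calf capsules supply body shapes, while a wide, thin root box forms a floor-like slab.
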PhysicsAsset /BRCosmetics/Characters/Player/Female/Medium/Bodies/F_MED_RankedOlympus/Meshes/F_MED_RankedOlympus_Physics
bodies {
 calf_l [{shapeType=Capsule, length=30, radius=8, center=<-20.5851 0.0001 0>, rotation=<rot -90 180 179.9998>}]
 calf_r [{shapeType=Capsule, length=30, radius=8, center=<20.5851 0.0001 0>, rotation=<rot -90 180 179.9998>}]
 dyn_cloth_b_1 [{shapeType=Capsule, length=10, radius=5, shapeName=dyn_cloth_b_1_capsule, center=<7.7908 0 2.8981>, rotation=<rot -0.0006 0 -90>}]
 dyn_cloth_b_2 [{shapeType=Capsule, length=10, radius=5, shapeName=dyn_cloth_b_2_capsule, center=<8.0732 0 2.0314>, rotation=<rot -0.0006 0 -90>}]
 dyn_cloth_b_3 [{shapeType=Capsule, length=7.5, radius=7.5, shapeName=dyn_cloth_b_3_capsule, center=<9.2805 0 3.9283>, rotation=<rot -0.0006 0 -90>}]
 dyn_cloth_fr_l_1 [{shapeType=Capsule, length=5, radius=2.5, shapeName=dyn_cloth_fr_l_1_capsule, center=<4.2674 -0.1272 1.1457>, rotation=<rot 90 0 -15>}]
 dyn_cloth_fr_l_2 [{shapeType=Capsule, length=11, radius=3.5, shapeName=dyn_cloth_fr_r_2_capsule, center=<7.9973 0.1807 1.1035>, rotation=<rot 90 0 0>}]
 dyn_cloth_fr_l_3 [{shapeType=Capsule, length=11, radius=3.3, shapeName=dyn_cloth_fr_r_3_capsule, center=<8.7468 0 1.935>, rotation=<rot 90 0 0>}]
 dyn_cloth_fr_l_4 [{shapeType=Capsule, length=6.5, radius=5.5, shapeName=dyn_cloth_fr_r_4_capsule, center=<10.7415 -0.4797 0.7153>, rotation=<rot 90 0 0>}]
 dyn_cloth_fr_r_1 [{shapeType=Capsule, length=5, radius=2.5, shapeName=dyn_cloth_fr_r_1_capsule, center=<4.2674 0.1272 1.1457>, rotation=<rot 90 -7.125 7.875>}]
 dyn_cloth_fr_r_2 [{shapeType=Capsule, length=11, radius=3.5, shapeName=dyn_cloth_fr_r_2_capsule, center=<7.9973 -0.1807 1.1035>, rotation=<rot 90 0 0>}]
 dyn_cloth_fr_r_3 [{shapeType=Capsule, length=11, radius=3.3, shapeName=dyn_cloth_fr_r_3_capsule, center=<8.7468 0 1.935>, rotation=<rot 90 0 0>}]
 dyn_cloth_fr_r_4 [{shapeType=Capsule, length=6.5, radius=5.5, shapeName=dyn_cloth_fr_r_4_capsule, center=<10.7415 0.5294 0.7153>, rotation=<rot 90 0 0>}]
 pelvis [{shapeType=Capsule, length=13.9389, radius=8.6537, center=<-7.3864 -2.8717 0>, rotation=<rot 0 -0.0006 0>}]
 root [{shapeType=Box, extentX=259.01, extentY=259.01, extentZ=12.01, shapeName=root_box, center=<0 0 -4.1074>}]
 thigh_l [{shapeType=Capsule, length=45.5, radius=8, center=<-20.5851 0 0>, rotation=<rot -85.9233 178.0788 178.7319>}]
 thigh_r [{shapeType=Capsule, length=45.5, radius=8, center=<20.5851 0 0>, rotation=<rot -85.9233 178.0788 178.7319>}]
}
constraints [
 {bone1Name=thigh_l, bone2Name=pelvis, swing1LimitDegrees=45, swing2LimitDegrees=45, twistLimitDegrees=45}
 {bone1Name=calf_l, bone2Name=thigh_l, swing1LimitDegrees=45, swing2LimitDegrees=45, twistLimitDegrees=45}
 {bone1Name=thigh_r, bone2Name=pelvis, swing1LimitDegrees=45, swing2LimitDegrees=45, twistLimitDegrees=45}
 {bone1Name=calf_r, bone2Name=thigh_r, swing1LimitDegrees=45, swing2LimitDegrees=45, twistLimitDegrees=45}
 {
  bone1Name=pelvis
  bone2Name=root
  swing1LimitDegrees=45
  swing1Motion=Free
  swing2LimitDegrees=45
  swing2Motion=Free
  twistLimitDegrees=45
  twistMotion=Free
 }
 {bone1Name=dyn_cloth_fr_l_1, bone2Name=pelvis, swing1LimitDegrees=20, swing2LimitDegrees=30, twistLimitDegrees=5}
 {bone1Name=dyn_cloth_fr_l_2, bone2Name=dyn_cloth_fr_l_1, swing1LimitDegrees=8, swing2LimitDegrees=25, twistLimitDegrees=4}
 {bone1Name=dyn_cloth_fr_l_3, bone2Name=dyn_cloth_fr_l_2, swing1LimitDegrees=8, swing2LimitDegrees=25, twistLimitDegrees=4}
 {bone1Name=dyn_cloth_fr_l_4, bone2Name=dyn_cloth_fr_l_3, swing1LimitDegrees=8, swing2LimitDegrees=25, twistLimitDegrees=4}
 {bone1Name=dyn_cloth_fr_r_1, bone2Name=pelvis, swing1LimitDegrees=20, swing2LimitDegrees=30, twistLimitDegrees=5}
 {bone1Name=dyn_cloth_fr_r_2, bone2Name=dyn_cloth_fr_r_1, swing1LimitDegrees=8, swing2LimitDegrees=25, twistLimitDegrees=4}
 {bone1Name=dyn_cloth_fr_r_3, bone2Name=dyn_cloth_fr_r_2, swing1LimitDegrees=8, swing2LimitDegrees=25, twistLimitDegrees=4}
 {bone1Name=dyn_cloth_fr_r_4, bone2Name=dyn_cloth_fr_r_3, swing1LimitDegrees=8, swing2LimitDegrees=25, twistLimitDegrees=4}
 {bone1Name=dyn_cloth_b_3, bone2Name=dyn_cloth_b_2, swing1LimitDegrees=20, swing2LimitDegrees=18, twistLimitDegrees=8}
 {bone1Name=dyn_cloth_b_2, bone2Name=dyn_cloth_b_1, swing1LimitDegrees=20, swing2LimitDegrees=18, twistLimitDegrees=8}
 {bone1Name=dyn_cloth_b_1, bone2Name=pelvis, swing1LimitDegrees=20, swing2LimitDegrees=14, twistLimitDegrees=8}
 {
  bone1Name=dyn_cloth_fr_r_4
  bone2Name=dyn_cloth_fr_l_4
  swing1LimitDegrees=45
  swing1Motion=Free
  swing2LimitDegrees=45
  swing2Motion=Free
  twistLimitDegrees=45
  twistMotion=Free
 }
 {
  bone1Name=dyn_cloth_fr_r_3
  bone2Name=dyn_cloth_fr_l_3
  swing1LimitDegrees=45
  swing1Motion=Free
  swing2LimitDegrees=45
  swing2Motion=Free
  twistLimitDegrees=45
  twistMotion=Free
 }
]
