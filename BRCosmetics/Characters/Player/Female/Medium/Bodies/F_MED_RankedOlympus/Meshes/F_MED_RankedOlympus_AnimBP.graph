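# Anim Blueprint for the female medium RankedOlympus body cosmetic, adding cloth and armor secondary motion to the pose copied from its parent mesh.
# A Control Rig feeds four RigidBody simulations with separate physics assets, selected by crouching, emote, front-end or standing state.
# Fixed additive offsets on cloth and pauldron bones finish the output.
# Updates compute gravity override and movement/bone alphas, but the movement-driven front-cloth chain is not wired to the output.
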AnimBlueprint /BRCosmetics/Characters/Player/Female/Medium/Bodies/F_MED_RankedOlympus/Meshes/F_MED_RankedOlympus_AnimBP
target_skeleton=&/F_MED_RankedOlympus_Skeleton
graphs [
 AnimGraph {nodes [
   n1 = "Transform (Modify) Bone - Bone: dyn_cloth_fr_MSTR"!(ComponentPose, bAlphaBoolEnabled:bool=True, Translation="0,-1,0", Rotation="0,0,0", Scale="1,1,1") -> (Pose) {
    alphaInputType=Bool
    rotationMode=BMM_Additive
    rotationSpace=BCS_BoneSpace
    translationMode=BMM_Additive
    translationSpace=BCS_BoneSpace
    alphaBoolBlend {blendInTime=0.3, blendOption=CircularInOut, blendOutTime=0.3}
    boneToModify {boneName=dyn_cloth_fr_MSTR}
    translation {y=-1}
   }
   n2 = "Transform (Modify) Bone - Bone: dyn_cloth_fr_MSTR"!(
    ComponentPose=@n1
    Alpha:float=Get Movement Backward Alpha() -> (Movement Backward Alpha:double)
    Translation="0,0,0"
    Rotation="0,0,0"
    Scale="1,1,1"
   ) -> (Pose) {
    rotationMode=BMM_Additive
    rotationSpace=BCS_BoneSpace
    translationMode=BMM_Additive
    translationSpace=BCS_BoneSpace
    alphaBoolBlend {blendInTime=0.3, blendOption=CircularInOut, blendOutTime=0.3}
    boneToModify {boneName=dyn_cloth_fr_MSTR}
   }
   n3 = "Transform (Modify) Bone - Bone: dyn_cloth_fr_MSTR"(
    ComponentPose=@n2
    Alpha=Get Movement Forward Alpha() -> (Movement Forward Alpha:double)
    Translation="0,0,0"
    Rotation="15,0,0"
    Scale="1,1,1"
   ) {
    rotationMode=BMM_Additive
    rotationSpace=BCS_BoneSpace
    translationMode=BMM_Additive
    translationSpace=BCS_BoneSpace
    alphaBoolBlend {blendInTime=0.3, blendOption=CircularInOut, blendOutTime=0.3}
    boneToModify {boneName=dyn_cloth_fr_MSTR}
    rotation {pitch=15}
   }
   n4 = "Transform (Modify) Bone - Bone: dyn_cloth_fr_MSTR"(
    ComponentPose=@n3
    Alpha=Get Movement to Left Alpha() -> (Movement to Left Alpha:double)
    Translation="0,0,0"
    Rotation="10,0,0"
    Scale="1,1,1"
   ) {
    rotationMode=BMM_Additive
    rotationSpace=BCS_BoneSpace
    translationMode=BMM_Additive
    translationSpace=BCS_BoneSpace
    alphaBoolBlend {blendInTime=0.3, blendOption=CircularInOut, blendOutTime=0.3}
    boneToModify {boneName=dyn_cloth_fr_MSTR}
    rotation {pitch=10}
   }
   "Transform (Modify) Bone - Bone: dyn_cloth_fr_MSTR"(
    ComponentPose=@n4
    Alpha=Get Movement to Right Alpha() -> (Movement to Right Alpha:double)
    Translation="0,0,0"
    Rotation="15,0,0"
    Scale="1,1,1"
   ) {
    rotationMode=BMM_Additive
    rotationSpace=BCS_BoneSpace
    translationMode=BMM_Additive
    translationSpace=BCS_BoneSpace
    alphaBoolBlend {blendInTime=0.3, blendOption=CircularInOut, blendOutTime=0.3}
    boneToModify {boneName=dyn_cloth_fr_MSTR}
    rotation {pitch=15}
   }
   n5 = "Transform (Modify) Bone - Bone: dyn_cloth_b_MSTR"!(
    ComponentPose=Local To Component(LocalPose=Use cached pose 'Post-Setup'())
    bAlphaBoolEnabled:bool=True
    Translation="-7.660,2.421,4.489"
    Rotation="79.886,4.951,14.835"
    Scale="1,1,1"
   ) -> (Pose) {
    alphaInputType=Bool
    rotationMode=BMM_Additive
    rotationSpace=BCS_BoneSpace
    translationMode=BMM_Additive
    translationSpace=BCS_BoneSpace
    alphaBoolBlend {blendInTime=0.3, blendOption=CircularInOut, blendOutTime=0.3}
    boneToModify {boneName=dyn_cloth_b_MSTR}
    rotation=<rot 79.886 4.951 14.835>
    translation=<-7.66 2.421 4.489>
   }
   n6 = "Transform (Modify) Bone - Bone: dyn_cloth_fr_r_MSTR"(ComponentPose=@n5, Alpha=1, Translation="-0.000,2.353,0.000", Rotation="0,0,0", Scale="1,1,1") {
    rotationMode=BMM_Additive
    rotationSpace=BCS_BoneSpace
    translationMode=BMM_Additive
    translationSpace=BCS_BoneSpace
    alphaBoolBlend {blendInTime=0.3, blendOption=CircularInOut, blendOutTime=0.3}
    boneToModify {boneName=dyn_cloth_fr_r_MSTR}
    translation=<0 2.3525 0>
   }
   n7 = "Transform (Modify) Bone - Bone: deform_pauldron_1"(ComponentPose=@n6, Alpha=1, Translation="0,0,-1", Rotation="3,0,-3", Scale="1,1,1") {
    rotationMode=BMM_Additive
    rotationSpace=BCS_BoneSpace
    translationMode=BMM_Additive
    translationSpace=BCS_BoneSpace
    alphaBoolBlend {blendInTime=0.3, blendOption=CircularInOut, blendOutTime=0.3}
    alphaScaleBiasClamp {bInterpResult=true}
    boneToModify {boneName=deform_pauldron_1}
    rotation {pitch=3, roll=-3}
    translation {z=-1}
   }
   Output Pose(Result=Component To Local(ComponentPose=@n7)) {layerGroup=None}
   n8 = Blend Poses by bool(BlendPose_0=Use cached pose 'FE cache'(), BlendPose_1=Use cached pose 'Standing cache'(), bActiveValue=Get bIsInFrontEnd()) {blendType=ExpInOut}
   n9 = Use cached pose 'Emoting cache'()
   n10 = Get bIsPlayingEmote()
   Save cached pose 'Post-Setup'(Pose=Blend Poses by bool(
     BlendPose_0=Use cached pose 'Crouching cache'()
     BlendPose_1=Blend Poses by bool(BlendPose_0=@n9, BlendPose_1=@n8, bActiveValue=@n10) {blendType=ExpInOut}
     bActiveValue=Get bIsCrouching()
    ) {blendType=ExpInOut}) {cacheName=Post-Setup}
   n11 = RigidBody!(
    ComponentPose=Local To Component(LocalPose=Use cached pose 'Post Control Rig'())
    Alpha
    OverrideWorldGravity:Vector=Get Gravity Override()
    ExternalForce
   ) -> (Pose) {
    bOverrideWorldGravity=true
    overridePhysicsAsset=&/F_MED_RankedOlympus_Physics
    simulationSpace=BaseBoneSpace
    alphaBoolBlend {blendOption=Linear}
    baseBoneRef {boneName=spine_01}
    componentAppliedLinearAccClamp=<3000 3000 3000>
    componentLinearAccScale=<0.3 0.3 0.3>
    componentLinearVelScale=<0.3 0.3 0.3>
   }
   Save cached pose 'Standing cache'(Pose=Component To Local(ComponentPose=@n11)) {cacheName=Standing cache}
   n12 = RigidBody(ComponentPose=Local To Component(LocalPose=Use cached pose 'Post Control Rig'()), OverrideWorldGravity=Get Gravity Override()) {
    bOverrideWorldGravity=true
    overridePhysicsAsset=&/F_MED_RankedOlympus_Emote_Physics
    simulationSpace=BaseBoneSpace
    alphaBoolBlend {blendOption=Linear}
    baseBoneRef {boneName=spine_01}
    componentAppliedLinearAccClamp=<3000 3000 3000>
    componentLinearAccScale=<0.3 0.3 0.3>
    componentLinearVelScale=<0.3 0.3 0.3>
   }
   Save cached pose 'Emoting cache'(Pose=Component To Local(ComponentPose=@n12)) {cacheName=Emoting cache}
   n13 = Control Rig(Source=Copy Pose From Mesh() {bUseAttachedParent=true}) {
    alphaBoolBlend {blendOption=Linear}
    controlRigAssetReference {blueprintRigClass=&/F_MED_RankedOlympus_CtrlRig.F_MED_RankedOlympus_CtrlRig_C}
   }
   Save cached pose 'Post Control Rig'(Pose=@n13) {cacheName=Post Control Rig}
   n14 = RigidBody(ComponentPose=Local To Component(LocalPose=Use cached pose 'Post Control Rig'()), OverrideWorldGravity=Get Gravity Override()) {
    bOverrideWorldGravity=true
    overridePhysicsAsset=&/F_MED_RankedOlympusCrouching_Physics
    simulationSpace=BaseBoneSpace
    alphaBoolBlend {blendOption=Linear}
    baseBoneRef {boneName=spine_01}
    componentAppliedLinearAccClamp=<3000 3000 3000>
    componentLinearAccScale=<0.3 0.3 0.3>
    componentLinearVelScale=<0.3 0.3 0.3>
   }
   Save cached pose 'Crouching cache'(Pose=Component To Local(ComponentPose=@n14)) {cacheName=Crouching cache}
   n15 = RigidBody(ComponentPose=Local To Component(LocalPose=Use cached pose 'Post Control Rig'()), OverrideWorldGravity=Get Gravity Override()) {
    bOverrideWorldGravity=true
    overridePhysicsAsset=&/F_MED_RankedOlympus_FE_Physics
    simulationSpace=BaseBoneSpace
    alphaBoolBlend {blendOption=Linear}
    baseBoneRef {boneName=spine_01}
    componentAppliedLinearAccClamp=<3000 3000 3000>
    componentLinearAccScale=<0.3 0.3 0.3>
    componentLinearVelScale=<0.3 0.3 0.3>
   }
   Save cached pose 'FE cache'(Pose=Component To Local(ComponentPose=@n15)) {cacheName=FE cache}
  ]}
 EventGraph {nodes [
   Event BlueprintUpdateAnimation()
   > n1 = Sequence!(execute) -> (then_0, then_1, then_2:Exec, then_3:Exec, then_4:Exec)
   > n2 = Gravity Override State Function DEPRECATED(CharacterPartAnimInstance=Self-Reference(), GameTimeInSeconds=GetGameTimeInSeconds(), GravityOverrideParams=Get Gravity Override Params())
   > Set Gravity Override(Gravity Override=@n2)
   n3 = Get Body Mesh() -> (Body Mesh:Skeletal Mesh Component Object Reference)
   n4 = NOT Boolean(A=Get bIsPlayingEmote())
   n5 = Get bIsInFrontEnd()
   Set FE shoulderpad(
    execute=@n1.then_4
    FE shoulderpad:double=float * float(
     A:bool=AND Boolean(A=@n5, B=@n4)
     B=GetAlphaFromReferenceTransformDelta(SkeletalMeshComponent=@n3, BoneName=upperarm_l, BaseName=spine_05, DeltaComponentType=EulerY, InputRange=(X=12,Y=19), TwistAxis="0,0,1")
    ) {is=title K2Node_PromotableOperator}
   ) -> (Output_Get:double)
   n6 = Get Body Mesh()
   Set Right Leg Bend Euler Z Alpha(
    execute=@n1.then_3
    Right Leg Bend Euler Z Alpha:double=float * float(
     A:double=GetAlphaFromReferenceTransformDelta(SkeletalMeshComponent=@n6, BoneName=thigh_r, BaseName=pelvis, InputRange=(X=-45,Y=-90), TwistAxis="0,0,1")
     B:bool=Get bIsPlayingEmote()
    )
   ) -> (Output_Get:double)
   n7 = Get bIsCrouching()
   n8 = TryGetPawnOwner()
   n9 = GetVelocity!(self:Actor Object Reference=@n8) -> (ReturnValue:Vector)
   n10 = Dot Product!(A:Vector=GetActorRightVector!(self:Actor Object Reference=@n8) -> (ReturnValue:Vector), B:Vector=@n9) -> (ReturnValue:double)
   n11 = Set Movement to Right Alpha(
    execute=@n1.then_1
    Movement to Right Alpha:double=float * float(A=MapRangeClamped(Value=@n10, InRangeA=0, InRangeB=230), B=NOT Boolean(A=@n7))
   ) -> (Output_Get:double)
   > Set Skydiving Movement to Right Alpha(Skydiving Movement to Right Alpha:double=float * float(A=@n11, B=Get bIsSkyDiving())) -> (Output_Get:double)
   n12 = Get bIsCrouching()
   > n13 = Set Movement to Left Alpha(Movement to Left Alpha:double=float * float(A=MapRangeClamped(Value=@n10, InRangeA=0, InRangeB=-230), B=NOT Boolean(A=@n12))) -> (Output_Get:double)
   > Set Skydiving Movement to Left Alpha(Skydiving Movement to Left Alpha:double=float * float(A=@n13, B=Get bIsSkyDiving())) -> (Output_Get:double)
   n14 = Get bIsCrouching()
   n15 = Dot Product(A=@n9, B=GetActorForwardVector!(self:Actor Object Reference=@n8) -> (ReturnValue:Vector))
   n16 = Set Movement Forward Alpha(
    execute=@n1.then_2
    Movement Forward Alpha:double=float * float(A=MapRangeClamped(Value=@n15, InRangeA=0, InRangeB=300), B=NOT Boolean(A=@n14))
   ) -> (Output_Get:double)
   > Set Skydiving Movement Forward Alpha(Skydiving Movement Forward Alpha:double=float * float(A=@n16, B=Get bIsSkyDiving())) -> (Output_Get:double)
   n17 = Get bIsCrouching()
   > n18 = Set Movement Backward Alpha(Movement Backward Alpha:double=float * float(A=MapRangeClamped(Value=@n15, InRangeA=0, InRangeB=-230), B=@n17)) -> (Output_Get:double)
   > Set Skydiving Movement Backward Alpha(Skydiving Movement Backward Alpha:double=float * float(A=@n18, B=Get bIsSkyDiving())) -> (Output_Get:double)
   Event OnCharacterCustomizationLoaded()
   > Set Body Mesh(Body Mesh:Skeletal Mesh Component Object Reference=GetOwnerSkeletalMeshComponent()) -> (Output_Get:Skeletal Mesh Component Object Reference)
   NOT Boolean(A=@n17)
   "Absolute (Float)"(A=0)
  ]}
]
variables [
 Body Mesh {type=SkeletalMeshComponent}
 FE shoulderpad {type=number}
 Gravity Override {type=Vector}
 Gravity Override Params {type=GravityOverrideParamsStruct}
 Movement Backward Alpha {type=number}
 Movement Forward Alpha {type=number}
 Movement to Left Alpha {type=number}
 Movement to Right Alpha {type=number}
 Right Leg Bend Euler Z Alpha {type=number}
 Skydiving Movement Backward Alpha {type=number}
 Skydiving Movement Forward Alpha {type=number}
 Skydiving Movement to Left Alpha {type=number}
 Skydiving Movement to Right Alpha {type=number}
]
