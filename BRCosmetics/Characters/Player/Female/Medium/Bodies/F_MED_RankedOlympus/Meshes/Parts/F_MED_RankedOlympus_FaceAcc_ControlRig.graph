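# Control Rig for the face accessory part of the RankedOlympus medium female body cosmetic, adjusting the right hair chain by head pose.
# A spherical pose reader on the head bone outputs a weight as the head's axis enters its rotated active and falloff regions.
# That weight blends a transform modification on dyn_hair_MSTR_R, the root of the right dynamic hair chain.
# Other hair chains are untouched, and the reader's debug draw is enabled.
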ControlRigBlueprint /BRCosmetics/Characters/Player/Female/Medium/Bodies/F_MED_RankedOlympus/Meshes/Parts/F_MED_RankedOlympus_FaceAcc_ControlRig
graphs [
 RigVMFunctionLibrary {links [], nodes []}
 RigVMModel {nodes [
   RigUnit_BeginExecution = &RigUnit_BeginExecution()
   n1 = &RigUnit_SphericalPoseReader(
    ActiveRegionScaleFactors=(PositiveWidth=1,NegativeWidth=1,PositiveHeight=1,NegativeHeight=1)
    ActiveRegionSize=0.05
    Debug=(bDrawDebug=true,bDraw2D=False,bDrawLocalAxes=False,DebugScale=25,DebugSegments=20,DebugThickness=0.25)
    DriverAxis=(X=1,Y=0,Z=0)
    DriverItem=(Type=Bone,Name="head")
    FalloffRegionScaleFactors=(PositiveWidth=1,NegativeWidth=1,PositiveHeight=1,NegativeHeight=1)
    FalloffSize=0.1
    FlipHeightScaling=False
    FlipWidthScaling=False
    OptionalParentItem=(Type=Bone,Name="None")
    RotationOffset=(X=0,Y=-80,Z=-20)
    ExecutePin=@RigUnit_BeginExecution.ExecutePin
   )
   > &RigUnit_ModifyTransforms(
    ItemToModify=((Item=(Type=Bone,Name="dyn_hair_MSTR_R"),Transform=(Rotation=(X=0.0076,Y=-0.816,Z=-0.0868,W=0.5714),Translation=(X=0,Y=0,Z=0),Scale3D=(X=1,Y=1,Z=1))))
    WeightMinimum=0
    Weight=@n1.OutputParam
   )
  ]}
]
hierarchy [
 root {loc=<0 0 0>, rot=<0 0 0 1>}
 pelvis {parent=root, loc=<0 -0.5301 98.0571>, rot=<0.0127 0.707 -0.0127 -0.707>}
 spine_01 {parent=pelvis, loc=<6.1208 -0.9328 0>, rot=<0 0 0.034 -0.9994>}
 spine_02 {parent=spine_01, loc=<6.5996 -0.2095 0>, rot=<0 0 0.0238 -0.9997>}
 spine_03 {parent=spine_02, loc=<6.4914 -0.1237 0>, rot=<0 0 0.0149 -0.9999>}
 spine_04 {parent=spine_03, loc=<6.9598 0.0794 0>, rot=<0 0 -0.0132 -0.9999>}
 spine_05 {parent=spine_04, loc=<6.9987 0 0>, rot=<0 0 -0.0015 -1>}
 neck_01 {parent=spine_05, loc=<11.3138 0.1529 0>, rot=<0 0 -0.2003 -0.9797>}
 neck_02 {parent=neck_01, loc=<4.5627 0 0>, rot=<0 0 0.0504 -0.9987>}
 head {parent=neck_02, loc=<5.3914 0 0>, rot=<0 0 0.115 -0.9934>}
 dyn_hair_MSTR_R {parent=head, loc=<9.6452 -4.5493 4.5014>, rot=<0.0042 -0.5175 0.8548 -0.0385>}
 dyn_hair_01_R {parent=dyn_hair_MSTR_R, loc=<0 0 0>, rot=<0 0 0 1>}
 dyn_hair_02_R {parent=dyn_hair_01_R, loc=<5.7 0 0>, rot=<0 0.0462 0 -0.9989>}
 dyn_hair_03_R {parent=dyn_hair_02_R, loc=<5.7 0 0>, rot=<0 0.272 0 -0.9623>}
 dyn_hair_04_R {parent=dyn_hair_03_R, loc=<5.7 0 0>, rot=<0 -0.0116 0 -0.9999>}
 dyn_hair_05_R {parent=dyn_hair_04_R, loc=<5.7 0 0>, rot=<0 -0.1196 0 -0.9928>}
 dyn_hair_06_R {parent=dyn_hair_05_R, loc=<5.7 0 0>, rot=<0 -0.0962 0 -0.9954>}
 dyn_hair_MSTR {parent=head, loc=<9.611 -5.669 0>, rot=<0.0272 -0.7066 0.7066 -0.0272>}
 dyn_hair_01 {parent=dyn_hair_MSTR, loc=<0 0 0>, rot=<0 0 0 1>}
 dyn_hair_02 {parent=dyn_hair_01, loc=<5.7 0 0>, rot=<0 0.1407 0 -0.9901>}
 dyn_hair_03 {parent=dyn_hair_02, loc=<5.7 0 0>, rot=<0 0.1289 0 -0.9917>}
 dyn_hair_04 {parent=dyn_hair_03, loc=<5.7 0 0>, rot=<0 -0.0454 0 -0.999>}
 dyn_hair_05 {parent=dyn_hair_04, loc=<5.7 0 0>, rot=<0 -0.1438 0 -0.9896>}
 dyn_hair_06 {parent=dyn_hair_05, loc=<5.7 0 0>, rot=<0 -0.0962 0 -0.9954>}
 dyn_hair_07 {parent=dyn_hair_06, loc=<5.7 0 0>, rot=<0 0 0 1>}
 dyn_hair_MSTR_L {parent=head, loc=<9.6447 -4.5493 -4.5014>, rot=<0.0208 -0.8419 0.5383 -0.0326>}
 dyn_hair_01_L {parent=dyn_hair_MSTR_L, loc=<0 0 0>, rot=<0 0 0 1>}
 dyn_hair_02_L {parent=dyn_hair_01_L, loc=<5.7 0 0>, rot=<0 0.1743 0 -0.9847>}
 dyn_hair_03_L {parent=dyn_hair_02_L, loc=<5.7 0 0>, rot=<0 0.1069 0 -0.9943>}
 dyn_hair_04_L {parent=dyn_hair_03_L, loc=<5.7 0 0>, rot=<0 -0.0116 0 -0.9999>}
 dyn_hair_05_L {parent=dyn_hair_04_L, loc=<5.7 0 0>, rot=<0 -0.1196 0 -0.9928>}
 dyn_hair_06_L {parent=dyn_hair_05_L, loc=<5.7 0 0>, rot=<0 -0.0962 0 -0.9954>}
 dyn_hair_top_MSTR {parent=head, loc=<14.3472 -3.9238 -0.0447>, rot=<-0.6602 0.2754 0.2798 -0.6403>}
 dyn_hair_top_01 {parent=dyn_hair_top_MSTR, loc=<0 0 0>, rot=<0 0 0 1>}
 dyn_hair_top_02 {parent=dyn_hair_top_01, loc=<5.5 0 0>, rot=<0 0.7567 0 -0.6537>}
 dyn_hair_top_03 {parent=dyn_hair_top_02, loc=<7 0 0>, rot=<0.0255 0.3424 -0.0697 -0.9366>}
 dyn_hair_top_04 {parent=dyn_hair_top_03, loc=<8 0 0>, rot=<0.0048 -0.0855 0.0564 -0.9947>}
 clavicle_l {parent=spine_05, loc=<5.4831 -0.1053 -3.2946>, rot=<0.0465 -0.7289 -0.009 -0.6829>}
 upperarm_l {parent=clavicle_l, loc=<9.8658 -1.2136 -0.4105>, rot=<0.0164 -0.3245 -0.0269 -0.9454>}
 clavicle_r {parent=spine_05, loc=<5.4831 -0.1053 3.2946>, rot=<0.7289 0.0465 -0.6829 0.009>}
 upperarm_r {parent=clavicle_r, loc=<-9.8658 1.2136 0.4105>, rot=<0.0164 -0.3245 -0.0269 -0.9454>}
 attach_backpack {parent=spine_05, loc=<-0.1228 -13.473 0>, rot=<-0.0092 -0.707 0.0092 -0.707>}
]
