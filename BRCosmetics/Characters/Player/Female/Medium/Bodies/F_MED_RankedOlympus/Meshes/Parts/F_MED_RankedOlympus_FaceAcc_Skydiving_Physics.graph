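# Physics asset for the skydiving variant of the RankedOlympus female face-accessory part, defining capsule bodies and constraints for its hair.
# Center, left, right and top hair chains each start at a MSTR bone, with limited swing and zero twist limits.
# All-free angular constraints cross-link side chains to the center, asymmetrically at the lower link, and left upper arm to clavicle.
# Head, neck, upper spine, clavicle and upper-arm capsules approximate the body.
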PhysicsAsset /BRCosmetics/Characters/Player/Female/Medium/Bodies/F_MED_RankedOlympus/Meshes/Parts/F_MED_RankedOlympus_FaceAcc_Skydiving_Physics
bodies {
 clavicle_l [{shapeType=Capsule, length=7.5, radius=8, shapeName=clavicle_l_capsule, center=<5 0 -2>, rotation=<rot -80 179.9991 -179.9997>}]
 clavicle_r [{shapeType=Capsule, length=7.5, radius=8, shapeName=clavicle_r_capsule, center=<-5 0 2>, rotation=<rot -80 179.999 -179.9996>}]
 dyn_hair_01 [{shapeType=Capsule, length=3, radius=2, shapeName=dyn_hair_01_capsule, center=<4.6067 0 -0.0001>, rotation=<rot -0.0009 0 -90>}]
 dyn_hair_01_L [{shapeType=Capsule, length=3, radius=2, shapeName=dyn_hair_01_L_capsule, center=<4.6067 0 -0.0001>, rotation=<rot -0.0008 0.0004 -90>}]
 dyn_hair_01_R [{shapeType=Capsule, length=3, radius=2, shapeName=dyn_hair_01_R_capsule, center=<4.6067 0 -0.0001>, rotation=<rot -0.0008 -0.0004 -90>}]
 dyn_hair_02 [{shapeType=Capsule, length=3, radius=2.5, shapeName=dyn_hair_02_capsule, center=<4.6067 0 -0.0001>, rotation=<rot -0.0009 0 -90>}]
 dyn_hair_02_L [{shapeType=Capsule, length=3, radius=2.5, shapeName=dyn_hair_02_L_capsule, center=<4.6067 0 -0.0001>, rotation=<rot -0.0008 0.0003 -90.0002>}]
 dyn_hair_02_R [{shapeType=Capsule, length=3, radius=2.5, shapeName=dyn_hair_02_R_capsule, center=<4.6067 0 -0.0001>, rotation=<rot -0.0008 -0.0004 -90>}]
 dyn_hair_03 [{shapeType=Capsule, length=3, radius=2.7, shapeName=dyn_hair_03_capsule, center=<4.6067 0 -0.0001>, rotation=<rot -0.0009 0 -90>}]
 dyn_hair_03_L [{shapeType=Capsule, length=3, radius=2.7, shapeName=dyn_hair_03_L_capsule, center=<4.6067 0 -0.0001>, rotation=<rot -0.0008 0.0003 -90.0002>}]
 dyn_hair_03_R [{shapeType=Capsule, length=3, radius=2.7, shapeName=dyn_hair_03_R_capsule, center=<4.6067 0 -0.0001>, rotation=<rot -0.0008 -0.0003 -89.9998>}]
 dyn_hair_04 [{shapeType=Capsule, length=3, radius=3, shapeName=dyn_hair_04_capsule, center=<4.6067 0 -0.0001>, rotation=<rot -0.0009 0 -90>}]
 dyn_hair_04_L [{shapeType=Capsule, length=3, radius=3, shapeName=dyn_hair_04_L_capsule, center=<4.6067 0 -0.0001>, rotation=<rot -0.0008 0.0003 -90.0002>}]
 dyn_hair_04_R [{shapeType=Capsule, length=3, radius=3, shapeName=dyn_hair_04_R_capsule, center=<4.6067 0 -0.0001>, rotation=<rot -0.0008 -0.0003 -89.9998>}]
 dyn_hair_05 [{shapeType=Capsule, length=3, radius=3, shapeName=dyn_hair_05_capsule, center=<4.6067 0 -0.0001>, rotation=<rot -0.0009 0 -90>}]
 dyn_hair_05_L [{shapeType=Capsule, length=3, radius=3, shapeName=dyn_hair_05_L_capsule, center=<4.6067 0 -0.0001>, rotation=<rot -0.0008 0.0003 -90.0001>}]
 dyn_hair_05_R [{shapeType=Capsule, length=3, radius=3, shapeName=dyn_hair_05_R_capsule, center=<4.6067 0 -0.0001>, rotation=<rot -0.0008 -0.0004 -89.9999>}]
 dyn_hair_06 [{shapeType=Capsule, length=3, radius=3, shapeName=dyn_hair_06_capsule, center=<4.6067 0 -0.0001>, rotation=<rot -0.0009 0 -90>}]
 dyn_hair_06_L [{shapeType=Capsule, length=3, radius=3, shapeName=dyn_hair_06_L_capsule, center=<4.6067 0 -0.0001>, rotation=<rot -0.0008 0.0004 -90.0001>}]
 dyn_hair_06_R [{shapeType=Capsule, length=3, radius=3, shapeName=dyn_hair_06_R_capsule, center=<4.6067 0 -0.0001>, rotation=<rot -0.0008 -0.0004 -89.9999>}]
 dyn_hair_07 [{shapeType=Capsule, length=3, radius=3, shapeName=dyn_hair_07_capsule, center=<4.6067 0 -0.0001>, rotation=<rot -0.0009 0 -90>}]
 dyn_hair_MSTR [{shapeType=Capsule, length=0, radius=0.5, shapeName=dyn_hair_MSTR_capsule}]
 dyn_hair_MSTR_L [{shapeType=Capsule, length=0, radius=0.5, shapeName=dyn_hair_MSTR_L_capsule}]
 dyn_hair_MSTR_R [{shapeType=Capsule, length=0, radius=0.5, shapeName=dyn_hair_MSTR_R_capsule}]
 dyn_hair_top_01 [{shapeType=Capsule, length=4.5, radius=2, shapeName=dyn_hair_top_03_capsule, center=<-0.446 0 0.2575>, rotation=<rot 60.0002 0 0>}]
 dyn_hair_top_02 [{shapeType=Capsule, length=4.5, radius=2, shapeName=dyn_hair_top_03_capsule, center=<1.5332 0 -0.5581>, rotation=<rot -70.0001 -180 -180>}]
 dyn_hair_top_03 [{shapeType=Capsule, length=5.5, radius=2, shapeName=dyn_hair_top_03_capsule, center=<4 0 0>, rotation=<rot -90 -90 90>}]
 dyn_hair_top_04 [{shapeType=Capsule, length=6, radius=2, shapeName=dyn_hair_top_04_capsule, center=<4 0 0>, rotation=<rot -80.0003 180 180>}]
 dyn_hair_top_MSTR [{shapeType=Capsule, length=0, radius=0.5, shapeName=dyn_hair_top_MSTR_capsule}]
 head [{shapeType=Capsule, length=4, radius=7.5, shapeName=head_capsule, center=<5.6874 3.1324 0>, rotation=<rot -90 0 -0.0002>}]
 neck_01 [{shapeType=Capsule, length=1, radius=7.5, shapeName=neck_01_capsule, center=<0 1.5432 0>, rotation=<rot -90 180 179.9998>}]
 neck_02 [{shapeType=Capsule, length=1, radius=7.5, shapeName=neck_02_capsule, center=<0 2.9048 0>, rotation=<rot -90 0.0199 -0.0201>}]
 spine_04 [{shapeType=Capsule, length=19, radius=8.5, shapeName=spine_04_capsule, center=<-10.5427 0 0>}]
 spine_05 [{shapeType=Capsule, length=19, radius=11, shapeName=spine_05_capsule, center=<0 1.1777 0>}]
 upperarm_l [{shapeType=Capsule, length=21, radius=9, shapeName=upperarm_l_capsule, center=<12.8269 0 3.0689>, rotation=<rot 90 0 -0.001>}]
 upperarm_r [{shapeType=Capsule, length=21, radius=9, shapeName=upperarm_r_capsule, center=<-12.8269 -0.0001 0>, rotation=<rot 90 179.7626 179.7623>}]
}
constraints [
 {bone1Name=dyn_hair_01, bone2Name=dyn_hair_MSTR, swing1LimitDegrees=10, swing2LimitDegrees=25, twistLimitDegrees=0}
 {bone1Name=dyn_hair_02, bone2Name=dyn_hair_01, swing1LimitDegrees=10, swing2LimitDegrees=25, twistLimitDegrees=0}
 {bone1Name=dyn_hair_03, bone2Name=dyn_hair_02, swing1LimitDegrees=10, swing2LimitDegrees=25, twistLimitDegrees=0}
 {bone1Name=dyn_hair_04, bone2Name=dyn_hair_03, swing1LimitDegrees=10, swing2LimitDegrees=25, twistLimitDegrees=0}
 {bone1Name=dyn_hair_05, bone2Name=dyn_hair_04, swing1LimitDegrees=10, swing2LimitDegrees=25, twistLimitDegrees=0}
 {bone1Name=dyn_hair_06, bone2Name=dyn_hair_05, swing1LimitDegrees=10, swing2LimitDegrees=25, twistLimitDegrees=0}
 {bone1Name=dyn_hair_07, bone2Name=dyn_hair_06, swing1LimitDegrees=10, swing2LimitDegrees=25, twistLimitDegrees=0}
 {bone1Name=dyn_hair_01_L, bone2Name=dyn_hair_MSTR_L, swing1LimitDegrees=10, swing2LimitDegrees=25, twistLimitDegrees=0}
 {bone1Name=dyn_hair_02_L, bone2Name=dyn_hair_01_L, swing1LimitDegrees=10, swing2LimitDegrees=25, twistLimitDegrees=0}
 {bone1Name=dyn_hair_03_L, bone2Name=dyn_hair_02_L, swing1LimitDegrees=10, swing2LimitDegrees=25, twistLimitDegrees=0}
 {bone1Name=dyn_hair_04_L, bone2Name=dyn_hair_03_L, swing1LimitDegrees=10, swing2LimitDegrees=25, twistLimitDegrees=0}
 {bone1Name=dyn_hair_05_L, bone2Name=dyn_hair_04_L, swing1LimitDegrees=10, swing2LimitDegrees=25, twistLimitDegrees=0}
 {bone1Name=dyn_hair_06_L, bone2Name=dyn_hair_05_L, swing1LimitDegrees=10, swing2LimitDegrees=25, twistLimitDegrees=0}
 {bone1Name=dyn_hair_01_R, bone2Name=dyn_hair_MSTR_R, swing1LimitDegrees=10, swing2LimitDegrees=25, twistLimitDegrees=0}
 {bone1Name=dyn_hair_02_R, bone2Name=dyn_hair_01_R, swing1LimitDegrees=10, swing2LimitDegrees=25, twistLimitDegrees=0}
 {bone1Name=dyn_hair_03_R, bone2Name=dyn_hair_02_R, swing1LimitDegrees=10, swing2LimitDegrees=25, twistLimitDegrees=0}
 {bone1Name=dyn_hair_04_R, bone2Name=dyn_hair_03_R, swing1LimitDegrees=10, swing2LimitDegrees=25, twistLimitDegrees=0}
 {bone1Name=dyn_hair_05_R, bone2Name=dyn_hair_04_R, swing1LimitDegrees=10, swing2LimitDegrees=25, twistLimitDegrees=0}
 {bone1Name=dyn_hair_06_R, bone2Name=dyn_hair_05_R, swing1LimitDegrees=10, swing2LimitDegrees=25, twistLimitDegrees=0}
 {bone1Name=dyn_hair_top_01, bone2Name=dyn_hair_top_MSTR, swing1LimitDegrees=0, swing2LimitDegrees=15, twistLimitDegrees=0}
 {bone1Name=dyn_hair_top_02, bone2Name=dyn_hair_top_01, swing1LimitDegrees=25, swing2LimitDegrees=45, twistLimitDegrees=0}
 {
  bone1Name=dyn_hair_05_R
  bone2Name=dyn_hair_05
  swing1LimitDegrees=45
  swing1Motion=Free
  swing2LimitDegrees=45
  swing2Motion=Free
  twistLimitDegrees=45
  twistMotion=Free
 }
 {
  bone1Name=dyn_hair_05_L
  bone2Name=dyn_hair_04
  swing1LimitDegrees=45
  swing1Motion=Free
  swing2LimitDegrees=45
  swing2Motion=Free
  twistLimitDegrees=45
  twistMotion=Free
 }
 {bone1Name=dyn_hair_top_03, bone2Name=dyn_hair_top_02, swing1LimitDegrees=10, swing2LimitDegrees=45, twistLimitDegrees=0}
 {bone1Name=dyn_hair_top_04, bone2Name=dyn_hair_top_03, swing1LimitDegrees=25, swing2LimitDegrees=45, twistLimitDegrees=0}
 {
  bone1Name=dyn_hair_01_L
  bone2Name=dyn_hair_01
  swing1LimitDegrees=45
  swing1Motion=Free
  swing2LimitDegrees=45
  swing2Motion=Free
  twistLimitDegrees=45
  twistMotion=Free
 }
 {
  bone1Name=dyn_hair_01_R
  bone2Name=dyn_hair_01
  swing1LimitDegrees=45
  swing1Motion=Free
  swing2LimitDegrees=45
  swing2Motion=Free
  twistLimitDegrees=45
  twistMotion=Free
 }
 {
  bone1Name=upperarm_l
  bone2Name=clavicle_l
  swing1LimitDegrees=45
  swing1Motion=Free
  swing2LimitDegrees=45
  swing2Motion=Free
  twistLimitDegrees=45
  twistMotion=Free
 }
]
