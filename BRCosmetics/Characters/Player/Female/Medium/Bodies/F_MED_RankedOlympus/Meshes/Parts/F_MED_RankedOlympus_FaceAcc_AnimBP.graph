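# Animation Blueprint for the RankedOlympus female FaceAcc part, simulating its dynamic hair bones.
# The copied parent pose, with upperarm_l matched to the body mesh, feeds three RigidBody branches: default (Control Rig for assault rifle/shotgun), backpack, and skydiving without backpack.
# The backpack branch adds bone constraints and head-rotation correctives; both skydiving-capable branches add pose offsets.
# Update logic derives branch flags, head-delta alphas and wind-driven gravity overrides.
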AnimBlueprint /BRCosmetics/Characters/Player/Female/Medium/Bodies/F_MED_RankedOlympus/Meshes/Parts/F_MED_RankedOlympus_FaceAcc_AnimBP
target_skeleton=&/F_MED_RankedOlympus_FaceAcc_Skeleton
graphs [
 AnimGraph {nodes [
   n1 = Control Rig(Source=Use cached pose 'main cache'(), Alpha=Get weapon() -> (weapon:double)) {
    alphaBoolBlend {blendOption=Linear}
    controlRigAssetReference {blueprintRigClass=&/F_MED_RankedOlympus_FaceAcc_ControlRig.F_MED_RankedOlympus_FaceAcc_ControlRig_C}
   }
   n2 = RigidBody!(ComponentPose=Local To Component(LocalPose=@n1), Alpha, OverrideWorldGravity:Vector=Get Gravity Override()) -> (Pose) {
    bClampLinearTranslationLimitToRefPose=true
    bOverrideWorldGravity=true
    overridePhysicsAsset=&/F_MED_RankedOlympus_FaceAcc_PhysicsOptimized
    simulationSpace=BaseBoneSpace
    alphaBoolBlend {blendOption=Linear}
    baseBoneRef {boneName=neck_01}
    componentAppliedLinearAccClamp=<900 800 900>
    componentLinearAccScale=<0.5 0.3 0.5>
    componentLinearVelScale=<0.5 0.35 0.5>
   }
   n3 = Component To Local(ComponentPose=@n2)
   n4 = Use cached pose 'is wearing backpack'()
   n5 = Get isWearingBackpack() -> (isWearingBackpack:bool)
   n6 = Get SkydiveDiveMode() -> (SkydiveDiveMode:double)
   n7 = "Transform (Modify) Bone - Bone: dyn_hair_MSTR_L"!(ComponentPose=Local To Component(LocalPose=Use cached pose 'main cache'()), Alpha=@n6, Rotation="-25,0,0") -> (Pose) {
    rotationMode=BMM_Additive
    rotationSpace=BCS_BoneSpace
    alphaBoolBlend {blendOption=Linear}
    alphaScaleBiasClamp {bInterpResult=true, interpSpeedDecreasing=5, interpSpeedIncreasing=5}
    boneToModify {boneName=dyn_hair_MSTR_L}
    rotation {pitch=-25}
   }
   n8 = "Transform (Modify) Bone - Bone: dyn_hair_MSTR"!(ComponentPose=@n7, Alpha=@n6, Rotation="-25,0,0") -> (Pose) {
    rotationMode=BMM_Additive
    rotationSpace=BCS_BoneSpace
    alphaBoolBlend {blendOption=Linear}
    alphaScaleBiasClamp {bInterpResult=true, interpSpeedDecreasing=5, interpSpeedIncreasing=5}
    boneToModify {boneName=dyn_hair_MSTR}
    rotation {pitch=-25}
   }
   n9 = "Transform (Modify) Bone - Bone: dyn_hair_MSTR_R"!(ComponentPose=@n8, Alpha=@n6, Rotation="-25,0,0") -> (Pose) {
    rotationMode=BMM_Additive
    rotationSpace=BCS_BoneSpace
    alphaBoolBlend {blendOption=Linear}
    alphaScaleBiasClamp {bInterpResult=true, interpSpeedDecreasing=5, interpSpeedIncreasing=5}
    boneToModify {boneName=dyn_hair_MSTR_R}
    rotation {pitch=-25}
   }
   n10 = RigidBody(ComponentPose=@n9, OverrideWorldGravity=Get Gravity Override()) {
    bClampLinearTranslationLimitToRefPose=true
    bOverrideWorldGravity=true
    overridePhysicsAsset=&/F_MED_RankedOlympus_FaceAcc_Skydiving_Physics
    simulationSpace=BaseBoneSpace
    alphaBoolBlend {blendOption=Linear}
    baseBoneRef {boneName=spine_05}
    componentAppliedLinearAccClamp=<900 900 900>
    componentLinearAccScale=<0.5 0.5 0.5>
    componentLinearVelScale=<0.5 0.5 0.5>
   }
   n11 = Get SkydiveFloating() -> (SkydiveFloating:double)
   n12 = "Transform (Modify) Bone - Bone: dyn_hair_04_L"!(ComponentPose=@n10, Alpha=@n11, Rotation="-30,0,0") -> (Pose) {
    rotationMode=BMM_Additive
    rotationSpace=BCS_BoneSpace
    alphaBoolBlend {blendOption=Linear}
    alphaScaleBiasClamp {bInterpResult=true, interpSpeedDecreasing=5, interpSpeedIncreasing=5}
    boneToModify {boneName=dyn_hair_04_L}
    rotation {pitch=-30}
   }
   n13 = "Transform (Modify) Bone - Bone: dyn_hair_04"!(ComponentPose=@n12, Alpha=@n11, Rotation="-30,0,0") -> (Pose) {
    rotationMode=BMM_Additive
    rotationSpace=BCS_BoneSpace
    alphaBoolBlend {blendOption=Linear}
    alphaScaleBiasClamp {bInterpResult=true, interpSpeedDecreasing=5, interpSpeedIncreasing=5}
    boneToModify {boneName=dyn_hair_04}
    rotation {pitch=-30}
   }
   n14 = "Transform (Modify) Bone - Bone: dyn_hair_04_R"!(ComponentPose=@n13, Alpha=@n11, Rotation="-30,0,0") -> (Pose) {
    rotationMode=BMM_Additive
    rotationSpace=BCS_BoneSpace
    alphaBoolBlend {blendOption=Linear}
    alphaScaleBiasClamp {bInterpResult=true, interpSpeedDecreasing=5, interpSpeedIncreasing=5}
    boneToModify {boneName=dyn_hair_04_R}
    rotation {pitch=-30}
   }
   Output Pose(Result=Blend Poses by bool(
     BlendPose_0=Component To Local(ComponentPose=@n14)
     BlendPose_1=Blend Poses by bool(BlendPose_0=@n4, BlendPose_1=@n3, bActiveValue=@n5) {blendType=Linear}
     bActiveValue=Get skydiving without backpack() -> (skydiving without backpack:bool)
    ) {blendType=Linear}) {layerGroup=None}
   n15 = "Transform (Modify) Bone - Bone: upperarm_l"!(
    ComponentPose=Local To Component(LocalPose=Copy Pose From Mesh() {bUseAttachedParent=true})
    Alpha=1
    Translation=Get translation() -> (translation:Vector)
    Rotation=Get Rotation() -> (Rotation:Rotator)
   ) -> (Pose) {
    rotationMode=BMM_Replace
    rotationSpace=BCS_WorldSpace
    translationMode=BMM_Replace
    translationSpace=BCS_WorldSpace
    alphaBoolBlend {blendOption=Linear}
    boneToModify {boneName=upperarm_l}
   }
   Save cached pose 'main cache'(Pose=Component To Local(ComponentPose=@n15)) {cacheName=main cache}
   n16 = "Constraint - dyn_hair_01_L (P)"(ComponentPose=Local To Component(LocalPose=Use cached pose 'main cache'()), Alpha=1, ConstraintWeights_0=1) {
    alphaBoolBlend {blendOption=Linear}
    boneToModify {boneName=dyn_hair_01_L}
    constraintSetup [{targetBone {boneName=head}}]
    constraintWeights [1]
   }
   n17 = "Constraint - dyn_hair_01 (P)"(ComponentPose=@n16, Alpha=1, ConstraintWeights_0=1) {
    alphaBoolBlend {blendOption=Linear}
    boneToModify {boneName=dyn_hair_01}
    constraintSetup [{targetBone {boneName=head}}]
    constraintWeights [1]
   }
   n18 = "Constraint - dyn_hair_01_R (P)"(ComponentPose=@n17, Alpha=1, ConstraintWeights_0=1) {
    alphaBoolBlend {blendOption=Linear}
    boneToModify {boneName=dyn_hair_01_R}
    constraintSetup [{targetBone {boneName=head}}]
    constraintWeights [1]
   }
   n19 = "Constraint - dyn_hair_02_L (P)"(ComponentPose=@n18, Alpha=1, ConstraintWeights_0=1) {
    alphaBoolBlend {blendOption=Linear}
    boneToModify {boneName=dyn_hair_02_L}
    constraintSetup [{targetBone {boneName=head}}]
    constraintWeights [1]
   }
   n20 = "Constraint - dyn_hair_02 (P)"(ComponentPose=@n19, Alpha=1, ConstraintWeights_0=1) {
    alphaBoolBlend {blendOption=Linear}
    boneToModify {boneName=dyn_hair_02}
    constraintSetup [{targetBone {boneName=head}}]
    constraintWeights [1]
   }
   n21 = "Constraint - dyn_hair_02_R (P)"(ComponentPose=@n20, Alpha=1, ConstraintWeights_0=1) {
    alphaBoolBlend {blendOption=Linear}
    boneToModify {boneName=dyn_hair_02_R}
    constraintSetup [{targetBone {boneName=head}}]
    constraintWeights [1]
   }
   n22 = "Constraint - dyn_hair_03_L (P)"(ComponentPose=@n21, Alpha=1, ConstraintWeights_0=1) {
    alphaBoolBlend {blendOption=Linear}
    boneToModify {boneName=dyn_hair_03_L}
    constraintSetup [{targetBone {boneName=spine_05}}]
    constraintWeights [1]
   }
   n23 = "Constraint - dyn_hair_03 (P)"(ComponentPose=@n22, Alpha=1, ConstraintWeights_0=1) {
    alphaBoolBlend {blendOption=Linear}
    boneToModify {boneName=dyn_hair_03}
    constraintSetup [{targetBone {boneName=spine_05}}]
    constraintWeights [1]
   }
   n24 = "Constraint - dyn_hair_03_R (P)"(ComponentPose=@n23, Alpha=1, ConstraintWeights_0=1) {
    alphaBoolBlend {blendOption=Linear}
    boneToModify {boneName=dyn_hair_03_R}
    constraintSetup [{targetBone {boneName=spine_05}}]
    constraintWeights [1]
   }
   n25 = "Constraint - dyn_hair_04_L (P)"(ComponentPose=@n24, Alpha=1, ConstraintWeights_0=1) {
    alphaBoolBlend {blendOption=Linear}
    boneToModify {boneName=dyn_hair_04_L}
    constraintSetup [{targetBone {boneName=spine_05}}]
    constraintWeights [1]
   }
   n26 = "Constraint - dyn_hair_04 (P)"(ComponentPose=@n25, Alpha=1, ConstraintWeights_0=1) {
    alphaBoolBlend {blendOption=Linear}
    boneToModify {boneName=dyn_hair_04}
    constraintSetup [{targetBone {boneName=spine_05}}]
    constraintWeights [1]
   }
   n27 = "Constraint - dyn_hair_04_R (P)"(ComponentPose=@n26, Alpha=1, ConstraintWeights_0=1) {
    alphaBoolBlend {blendOption=Linear}
    boneToModify {boneName=dyn_hair_04_R}
    constraintSetup [{targetBone {boneName=spine_05}}]
    constraintWeights [1]
   }
   n28 = "Constraint - dyn_hair_05_L (P)"(ComponentPose=@n27, Alpha=1, ConstraintWeights_0=1) {
    alphaBoolBlend {blendOption=Linear}
    boneToModify {boneName=dyn_hair_05_L}
    constraintSetup [{targetBone {boneName=spine_05}}]
    constraintWeights [1]
   }
   n29 = "Constraint - dyn_hair_05 (P)"(ComponentPose=@n28, Alpha=1, ConstraintWeights_0=1) {
    alphaBoolBlend {blendOption=Linear}
    boneToModify {boneName=dyn_hair_05}
    constraintSetup [{targetBone {boneName=spine_05}}]
    constraintWeights [1]
   }
   n30 = "Constraint - dyn_hair_05_R (P)"(ComponentPose=@n29, Alpha=1, ConstraintWeights_0=1) {
    alphaBoolBlend {blendOption=Linear}
    boneToModify {boneName=dyn_hair_05_R}
    constraintSetup [{targetBone {boneName=spine_05}}]
    constraintWeights [1]
   }
   n31 = "Constraint - dyn_hair_06_L (P)"(ComponentPose=@n30, Alpha=1, ConstraintWeights_0=1) {
    alphaBoolBlend {blendOption=Linear}
    boneToModify {boneName=dyn_hair_06_L}
    constraintSetup [{targetBone {boneName=spine_05}}]
    constraintWeights [1]
   }
   n32 = "Constraint - dyn_hair_06 (P)"(ComponentPose=@n31, Alpha=1, ConstraintWeights_0=1) {
    alphaBoolBlend {blendOption=Linear}
    boneToModify {boneName=dyn_hair_06}
    constraintSetup [{targetBone {boneName=spine_05}}]
    constraintWeights [1]
   }
   n33 = "Constraint - dyn_hair_06_R (P)"(ComponentPose=@n32, Alpha=1, ConstraintWeights_0=1) {
    alphaBoolBlend {blendOption=Linear}
    boneToModify {boneName=dyn_hair_06_R}
    constraintSetup [{targetBone {boneName=spine_05}}]
    constraintWeights [1]
   }
   n34 = "Constraint - dyn_hair_07 (P)"(ComponentPose=@n33, Alpha=1, ConstraintWeights_0=1) {
    alphaBoolBlend {blendOption=Linear}
    boneToModify {boneName=dyn_hair_07}
    constraintSetup [{targetBone {boneName=spine_05}}]
    constraintWeights [1]
   }
   n35 = Get head pitch1() -> (head pitch1:double)
   n36 = "Transform (Modify) Bone - Bone: dyn_hair_01_L"!(ComponentPose=@n34, Alpha=@n35, Translation="0,0,0", Rotation="44.803,0,0") -> (Pose) {
    rotationMode=BMM_Additive
    rotationSpace=BCS_BoneSpace
    translationMode=BMM_Additive
    translationSpace=BCS_BoneSpace
    alphaBoolBlend {blendOption=Linear}
    alphaScaleBiasClamp {bInterpResult=true}
    boneToModify {boneName=dyn_hair_01_L}
    rotation {pitch=44.803}
   }
   n37 = "Transform (Modify) Bone - Bone: dyn_hair_01"!(ComponentPose=@n36, Alpha=@n35, Translation="0,0,0", Rotation="45,0.003,0.012") -> (Pose) {
    rotationMode=BMM_Additive
    rotationSpace=BCS_BoneSpace
    translationMode=BMM_Additive
    translationSpace=BCS_BoneSpace
    alphaBoolBlend {blendOption=Linear}
    alphaScaleBiasClamp {bInterpResult=true}
    boneToModify {boneName=dyn_hair_01}
    rotation=<rot 45 0.003 0.012>
   }
   n38 = "Transform (Modify) Bone - Bone: dyn_hair_01_R"!(ComponentPose=@n37, Alpha=@n35, Translation="0,0,0", Rotation="45.463,27.170,27.341") -> (Pose) {
    rotationMode=BMM_Additive
    rotationSpace=BCS_BoneSpace
    translationMode=BMM_Additive
    translationSpace=BCS_BoneSpace
    alphaBoolBlend {blendOption=Linear}
    alphaScaleBiasClamp {bInterpResult=true}
    boneToModify {boneName=dyn_hair_01_R}
    rotation=<rot 45.4628 27.1699 27.3408>
   }
   n39 = "Transform (Modify) Bone - Bone: dyn_hair_03_L"!(ComponentPose=@n38, Alpha=@n35, Translation="0,0,0", Rotation="-55,0,0") -> (Pose) {
    rotationMode=BMM_Additive
    rotationSpace=BCS_BoneSpace
    translationMode=BMM_Additive
    translationSpace=BCS_BoneSpace
    alphaBoolBlend {blendOption=Linear}
    alphaScaleBiasClamp {bInterpResult=true}
    boneToModify {boneName=dyn_hair_03_L}
    rotation {pitch=-55}
   }
   n40 = "Transform (Modify) Bone - Bone: dyn_hair_03"!(ComponentPose=@n39, Alpha=@n35, Translation="0,0,0", Rotation="-16.201,-8.324,-2.277") -> (Pose) {
    rotationMode=BMM_Additive
    rotationSpace=BCS_BoneSpace
    translationMode=BMM_Additive
    translationSpace=BCS_BoneSpace
    alphaBoolBlend {blendOption=Linear}
    alphaScaleBiasClamp {bInterpResult=true}
    boneToModify {boneName=dyn_hair_03}
    rotation=<rot -16.2014 -8.3239 -2.2771>
   }
   n41 = "Transform (Modify) Bone - Bone: dyn_hair_03_R"!(ComponentPose=@n40, Alpha=@n35, Translation="0,0,0", Rotation="-75,0,0") -> (Pose) {
    rotationMode=BMM_Additive
    rotationSpace=BCS_BoneSpace
    translationMode=BMM_Additive
    translationSpace=BCS_BoneSpace
    alphaBoolBlend {blendOption=Linear}
    alphaScaleBiasClamp {bInterpResult=true}
    boneToModify {boneName=dyn_hair_03_R}
    rotation {pitch=-75}
   }
   n42 = "Transform (Modify) Bone - Bone: dyn_hair_04_L"!(ComponentPose=@n41, Alpha=@n35, Translation:Vector="0,0,0", Rotation="-5.23,0,0") -> (Pose) {
    rotationMode=BMM_Additive
    rotationSpace=BCS_BoneSpace
    translationMode=BMM_Additive
    translationSpace=BCS_BoneSpace
    alphaBoolBlend {blendOption=Linear}
    alphaScaleBiasClamp {bInterpResult=true}
    boneToModify {boneName=dyn_hair_04_L}
    rotation {pitch=-5.23}
   }
   n43 = "Transform (Modify) Bone - Bone: dyn_hair_04"!(ComponentPose=@n42, Alpha=@n35, Translation:Vector="0,0,0", Rotation="-45,0,0") -> (Pose) {
    rotationMode=BMM_Additive
    rotationSpace=BCS_BoneSpace
    translationMode=BMM_Additive
    translationSpace=BCS_BoneSpace
    alphaBoolBlend {blendOption=Linear}
    alphaScaleBiasClamp {bInterpResult=true}
    boneToModify {boneName=dyn_hair_04}
    rotation {pitch=-45}
   }
   n44 = "Transform (Modify) Bone - Bone: dyn_hair_04_R"!(ComponentPose=@n43, Alpha=@n35, Translation:Vector="0,0,0", Rotation="-10,0,0") -> (Pose) {
    rotationMode=BMM_Additive
    rotationSpace=BCS_BoneSpace
    translationMode=BMM_Additive
    translationSpace=BCS_BoneSpace
    alphaBoolBlend {blendOption=Linear}
    alphaScaleBiasClamp {bInterpResult=true}
    boneToModify {boneName=dyn_hair_04_R}
    rotation {pitch=-10}
   }
   n45 = Get head pitch2() -> (head pitch2:float)
   n46 = "Transform (Modify) Bone - Bone: dyn_hair_01_L"(ComponentPose=@n44, Alpha=@n45, Translation="0,0,0", Rotation="40,0,0") {
    rotationMode=BMM_Additive
    rotationSpace=BCS_BoneSpace
    translationMode=BMM_Additive
    translationSpace=BCS_BoneSpace
    alphaBoolBlend {blendOption=Linear}
    boneToModify {boneName=dyn_hair_01_L}
    rotation {pitch=40}
   }
   n47 = "Transform (Modify) Bone - Bone: dyn_hair_01"(ComponentPose=@n46, Alpha=@n45, Translation="0,0,0", Rotation="30,0,0") {
    rotationMode=BMM_Additive
    rotationSpace=BCS_BoneSpace
    translationMode=BMM_Additive
    translationSpace=BCS_BoneSpace
    alphaBoolBlend {blendOption=Linear}
    boneToModify {boneName=dyn_hair_01}
    rotation {pitch=30}
   }
   n48 = "Transform (Modify) Bone - Bone: dyn_hair_01_R"(ComponentPose=@n47, Alpha=@n45, Translation="0,0,0", Rotation="30,0,0") {
    rotationMode=BMM_Additive
    rotationSpace=BCS_BoneSpace
    translationMode=BMM_Additive
    translationSpace=BCS_BoneSpace
    alphaBoolBlend {blendOption=Linear}
    boneToModify {boneName=dyn_hair_01_R}
    rotation {pitch=30}
   }
   n49 = "Transform (Modify) Bone - Bone: dyn_hair_03_L"(ComponentPose=@n48, Alpha=@n45, Translation="0,0,0", Rotation="-30,0,0") {
    rotationMode=BMM_Additive
    rotationSpace=BCS_BoneSpace
    translationMode=BMM_Additive
    translationSpace=BCS_BoneSpace
    alphaBoolBlend {blendOption=Linear}
    boneToModify {boneName=dyn_hair_03_L}
    rotation {pitch=-30}
   }
   n50 = "Transform (Modify) Bone - Bone: dyn_hair_03"(ComponentPose=@n49, Alpha=@n45, Translation="0,0,0", Rotation="-30,0,0") {
    rotationMode=BMM_Additive
    rotationSpace=BCS_BoneSpace
    translationMode=BMM_Additive
    translationSpace=BCS_BoneSpace
    alphaBoolBlend {blendOption=Linear}
    boneToModify {boneName=dyn_hair_03}
    rotation {pitch=-30}
   }
   n51 = "Transform (Modify) Bone - Bone: dyn_hair_03_R"(ComponentPose=@n50, Alpha=@n45, Translation="0,0,0", Rotation="-30,0,0") {
    rotationMode=BMM_Additive
    rotationSpace=BCS_BoneSpace
    translationMode=BMM_Additive
    translationSpace=BCS_BoneSpace
    alphaBoolBlend {blendOption=Linear}
    boneToModify {boneName=dyn_hair_03_R}
    rotation {pitch=-30}
   }
   n52 = "Transform (Modify) Bone - Bone: dyn_hair_04_L"(ComponentPose=@n51, Alpha=@n45, Translation="0,0,0", Rotation="0,0,0") {
    rotationMode=BMM_Additive
    rotationSpace=BCS_BoneSpace
    translationMode=BMM_Additive
    translationSpace=BCS_BoneSpace
    alphaBoolBlend {blendOption=Linear}
    boneToModify {boneName=dyn_hair_04_L}
   }
   n53 = "Transform (Modify) Bone - Bone: dyn_hair_04"(ComponentPose=@n52, Alpha=@n45, Translation="0,0,0", Rotation="0,0,0") {
    rotationMode=BMM_Additive
    rotationSpace=BCS_BoneSpace
    translationMode=BMM_Additive
    translationSpace=BCS_BoneSpace
    alphaBoolBlend {blendOption=Linear}
    boneToModify {boneName=dyn_hair_04}
   }
   n54 = "Transform (Modify) Bone - Bone: dyn_hair_04_R"(ComponentPose=@n53, Alpha=@n45, Translation="0,0,0", Rotation="0,0,0") {
    rotationMode=BMM_Additive
    rotationSpace=BCS_BoneSpace
    translationMode=BMM_Additive
    translationSpace=BCS_BoneSpace
    alphaBoolBlend {blendOption=Linear}
    boneToModify {boneName=dyn_hair_04_R}
   }
   n55 = Get head roll() -> (head roll:float)
   n56 = "Transform (Modify) Bone - Bone: dyn_hair_01_L"(ComponentPose=@n54, Alpha=@n55, Translation="0,0,0", Rotation="-5,0,0") {
    rotationMode=BMM_Additive
    rotationSpace=BCS_BoneSpace
    translationMode=BMM_Additive
    translationSpace=BCS_BoneSpace
    alphaBoolBlend {blendOption=Linear}
    alphaScaleBiasClamp {bInterpResult=true}
    boneToModify {boneName=dyn_hair_01_L}
    rotation {pitch=-5}
   }
   n57 = "Transform (Modify) Bone - Bone: dyn_hair_01"(ComponentPose=@n56, Alpha=@n55, Translation="0,0,0", Rotation="35,0,0") {
    rotationMode=BMM_Additive
    rotationSpace=BCS_BoneSpace
    translationMode=BMM_Additive
    translationSpace=BCS_BoneSpace
    alphaBoolBlend {blendOption=Linear}
    alphaScaleBiasClamp {bInterpResult=true}
    boneToModify {boneName=dyn_hair_01}
    rotation {pitch=35}
   }
   n58 = "Transform (Modify) Bone - Bone: dyn_hair_01_R"(ComponentPose=@n57, Alpha=@n55, Translation="0,0,0", Rotation="65.000,0.028,0.090") {
    rotationMode=BMM_Additive
    rotationSpace=BCS_BoneSpace
    translationMode=BMM_Additive
    translationSpace=BCS_BoneSpace
    alphaBoolBlend {blendOption=Linear}
    alphaScaleBiasClamp {bInterpResult=true}
    boneToModify {boneName=dyn_hair_01_R}
    rotation=<rot 64.9999 0.0279 0.0899>
   }
   n59 = "Transform (Modify) Bone - Bone: dyn_hair_03_L"(ComponentPose=@n58, Alpha=@n55, Translation="0,0,0", Rotation="0,0,0") {
    rotationMode=BMM_Additive
    rotationSpace=BCS_BoneSpace
    translationMode=BMM_Additive
    translationSpace=BCS_BoneSpace
    alphaBoolBlend {blendOption=Linear}
    alphaScaleBiasClamp {bInterpResult=true}
    boneToModify {boneName=dyn_hair_03_L}
   }
   n60 = "Transform (Modify) Bone - Bone: dyn_hair_03"(ComponentPose=@n59, Alpha=@n55, Translation="0,0,0", Rotation="-44.388,41.395,-23.796") {
    rotationMode=BMM_Additive
    rotationSpace=BCS_BoneSpace
    translationMode=BMM_Additive
    translationSpace=BCS_BoneSpace
    alphaBoolBlend {blendOption=Linear}
    alphaScaleBiasClamp {bInterpResult=true}
    boneToModify {boneName=dyn_hair_03}
    rotation=<rot -44.3881 41.3951 -23.7961>
   }
   n61 = "Transform (Modify) Bone - Bone: dyn_hair_03_R"(ComponentPose=@n60, Alpha=@n55, Translation="0,0,0", Rotation="-39.201,44.260,-43.476") {
    rotationMode=BMM_Additive
    rotationSpace=BCS_BoneSpace
    translationMode=BMM_Additive
    translationSpace=BCS_BoneSpace
    alphaBoolBlend {blendOption=Linear}
    alphaScaleBiasClamp {bInterpResult=true}
    boneToModify {boneName=dyn_hair_03_R}
    rotation=<rot -39.2011 44.2601 -43.4761>
   }
   n62 = "Transform (Modify) Bone - Bone: dyn_hair_04_L"(ComponentPose=@n61, Alpha=@n55, Translation="0,0,0", Rotation="0,0,0") {
    rotationMode=BMM_Additive
    rotationSpace=BCS_BoneSpace
    translationMode=BMM_Additive
    translationSpace=BCS_BoneSpace
    alphaBoolBlend {blendOption=Linear}
    alphaScaleBiasClamp {bInterpResult=true}
    boneToModify {boneName=dyn_hair_04_L}
   }
   n63 = "Transform (Modify) Bone - Bone: dyn_hair_04"(ComponentPose=@n62, Alpha=@n55, Translation="0,0,0", Rotation="0,0,0") {
    rotationMode=BMM_Additive
    rotationSpace=BCS_BoneSpace
    translationMode=BMM_Additive
    translationSpace=BCS_BoneSpace
    alphaBoolBlend {blendOption=Linear}
    alphaScaleBiasClamp {bInterpResult=true}
    boneToModify {boneName=dyn_hair_04}
   }
   n64 = "Transform (Modify) Bone - Bone: dyn_hair_04_R"(ComponentPose=@n63, Alpha=@n55, Translation="0,0,0", Rotation="-30,0,0") {
    rotationMode=BMM_Additive
    rotationSpace=BCS_BoneSpace
    translationMode=BMM_Additive
    translationSpace=BCS_BoneSpace
    alphaBoolBlend {blendOption=Linear}
    alphaScaleBiasClamp {bInterpResult=true}
    boneToModify {boneName=dyn_hair_04_R}
    rotation {pitch=-30}
   }
   Save cached pose 'presim cache'(Pose=Component To Local(ComponentPose=@n64)) {cacheName=presim cache}
   n65 = RigidBody(
    ComponentPose=Local To Component(LocalPose=Use cached pose 'presim cache'())
    OverrideWorldGravity="Get backpack Gravity Override "() -> ("backpack Gravity Override ":Vector)
   ) {
    bClampLinearTranslationLimitToRefPose=true
    bOverrideWorldGravity=true
    overridePhysicsAsset=&/F_MED_RankedOlympus_FaceAcc_Backpack_Physics
    simulationSpace=BaseBoneSpace
    alphaBoolBlend {blendOption=Linear}
    baseBoneRef {boneName=spine_05}
    componentAppliedLinearAccClamp=<900 900 900>
    componentLinearAccScale=<0.5 0.5 0.5>
    componentLinearVelScale=<0.5 0.5 0.5>
   }
   n66 = Get SkydiveFloating()
   n67 = "Transform (Modify) Bone - Bone: dyn_hair_03_L"!(ComponentPose=@n65, Alpha=@n66, Rotation="20,0,0") -> (Pose) {
    rotationMode=BMM_Additive
    rotationSpace=BCS_BoneSpace
    alphaBoolBlend {blendOption=Linear}
    boneToModify {boneName=dyn_hair_03_L}
    rotation {pitch=20}
   }
   n68 = "Transform (Modify) Bone - Bone: dyn_hair_04_L"!(ComponentPose=@n67, Alpha=@n66, Rotation="-40,0,0") -> (Pose) {
    rotationMode=BMM_Additive
    rotationSpace=BCS_BoneSpace
    translationMode=BMM_Additive
    translationSpace=BCS_BoneSpace
    alphaBoolBlend {blendOption=Linear}
    boneToModify {boneName=dyn_hair_04_L}
    rotation {pitch=-40}
   }
   n69 = Get parachute opened() -> (parachute opened:double)
   n70 = "Transform (Modify) Bone - Bone: dyn_hair_03_L"(ComponentPose=@n68, Alpha=@n69, Rotation="-10,0,0") {
    rotationMode=BMM_Additive
    rotationSpace=BCS_BoneSpace
    alphaBoolBlend {blendOption=Linear}
    boneToModify {boneName=dyn_hair_03_L}
    rotation {pitch=-10}
   }
   n71 = Get not skydiving() -> (not skydiving:double)
   n72 = "Transform (Modify) Bone - Bone: dyn_hair_04_L"(ComponentPose=@n70, Alpha=@n71, Rotation="-5,0,0") {
    rotationMode=BMM_Additive
    rotationSpace=BCS_BoneSpace
    translationMode=BMM_Additive
    translationSpace=BCS_BoneSpace
    alphaBoolBlend {blendOption=Linear}
    boneToModify {boneName=dyn_hair_04_L}
    rotation {pitch=-5}
   }
   n73 = Get head yaw left skydiving() -> (head yaw left skydiving:float)
   n74 = "Transform (Modify) Bone - Bone: dyn_hair_01_L"!(ComponentPose=@n72, Alpha=@n73, Rotation="0,0,0") -> (Pose) {
    rotationMode=BMM_Additive
    rotationSpace=BCS_BoneSpace
    translationMode=BMM_Additive
    translationSpace=BCS_BoneSpace
    alphaBoolBlend {blendOption=Linear}
    boneToModify {boneName=dyn_hair_01_L}
   }
   n75 = Get head yaw right skydiving() -> (head yaw right skydiving:double)
   n76 = "Transform (Modify) Bone - Bone: dyn_hair_01_L"(ComponentPose=@n74, Alpha=@n75, Rotation="0,-40,0") {
    rotationMode=BMM_Additive
    rotationSpace=BCS_BoneSpace
    translationMode=BMM_Additive
    translationSpace=BCS_BoneSpace
    alphaBoolBlend {blendOption=Linear}
    boneToModify {boneName=dyn_hair_01_L}
    rotation {yaw=-40}
   }
   n77 = "Transform (Modify) Bone - Bone: dyn_hair_03"!(ComponentPose=@n76, Alpha=@n66, Rotation="20,0,0") -> (Pose) {
    rotationMode=BMM_Additive
    rotationSpace=BCS_BoneSpace
    alphaBoolBlend {blendOption=Linear}
    boneToModify {boneName=dyn_hair_03}
    rotation {pitch=20}
   }
   n78 = "Transform (Modify) Bone - Bone: dyn_hair_04"!(ComponentPose=@n77, Alpha=@n66, Rotation="-40,0,0") -> (Pose) {
    rotationMode=BMM_Additive
    rotationSpace=BCS_BoneSpace
    translationMode=BMM_Additive
    translationSpace=BCS_BoneSpace
    alphaBoolBlend {blendOption=Linear}
    boneToModify {boneName=dyn_hair_04}
    rotation {pitch=-40}
   }
   n79 = "Transform (Modify) Bone - Bone: dyn_hair_03"(ComponentPose=@n78, Alpha=@n69, Rotation="-10,0,0") {
    rotationMode=BMM_Additive
    rotationSpace=BCS_BoneSpace
    alphaBoolBlend {blendOption=Linear}
    boneToModify {boneName=dyn_hair_03}
    rotation {pitch=-10}
   }
   n80 = "Transform (Modify) Bone - Bone: dyn_hair_04"(ComponentPose=@n79, Alpha=@n71, Rotation="-5,0,0") {
    rotationMode=BMM_Additive
    rotationSpace=BCS_BoneSpace
    translationMode=BMM_Additive
    translationSpace=BCS_BoneSpace
    alphaBoolBlend {blendOption=Linear}
    boneToModify {boneName=dyn_hair_04}
    rotation {pitch=-5}
   }
   n81 = "Transform (Modify) Bone - Bone: dyn_hair_01"!(ComponentPose=@n80, Alpha=@n73, Rotation="0,20,0") -> (Pose) {
    rotationMode=BMM_Additive
    rotationSpace=BCS_BoneSpace
    translationMode=BMM_Additive
    translationSpace=BCS_BoneSpace
    alphaBoolBlend {blendOption=Linear}
    boneToModify {boneName=dyn_hair_01}
    rotation {yaw=20}
   }
   n82 = "Transform (Modify) Bone - Bone: dyn_hair_01"(ComponentPose=@n81, Alpha=@n75, Rotation="0,-20,0") {
    rotationMode=BMM_Additive
    rotationSpace=BCS_BoneSpace
    translationMode=BMM_Additive
    translationSpace=BCS_BoneSpace
    alphaBoolBlend {blendOption=Linear}
    boneToModify {boneName=dyn_hair_01}
    rotation {yaw=-20}
   }
   n83 = "Transform (Modify) Bone - Bone: dyn_hair_03_R"!(ComponentPose=@n82, Alpha=@n66, Rotation="20,0,0") -> (Pose) {
    rotationMode=BMM_Additive
    rotationSpace=BCS_BoneSpace
    alphaBoolBlend {blendOption=Linear}
    boneToModify {boneName=dyn_hair_03_R}
    rotation {pitch=20}
   }
   n84 = "Transform (Modify) Bone - Bone: dyn_hair_04_R"!(ComponentPose=@n83, Alpha=@n66, Rotation="-40,0,0") -> (Pose) {
    rotationMode=BMM_Additive
    rotationSpace=BCS_BoneSpace
    translationMode=BMM_Additive
    translationSpace=BCS_BoneSpace
    alphaBoolBlend {blendOption=Linear}
    boneToModify {boneName=dyn_hair_04_R}
    rotation {pitch=-40}
   }
   n85 = "Transform (Modify) Bone - Bone: dyn_hair_03_R"(ComponentPose=@n84, Alpha=@n69, Rotation="-10,0,0") {
    rotationMode=BMM_Additive
    rotationSpace=BCS_BoneSpace
    alphaBoolBlend {blendOption=Linear}
    boneToModify {boneName=dyn_hair_03_R}
    rotation {pitch=-10}
   }
   n86 = "Transform (Modify) Bone - Bone: dyn_hair_04_R"(ComponentPose=@n85, Alpha=@n71, Rotation="-5,0,0") {
    rotationMode=BMM_Additive
    rotationSpace=BCS_BoneSpace
    translationMode=BMM_Additive
    translationSpace=BCS_BoneSpace
    alphaBoolBlend {blendOption=Linear}
    boneToModify {boneName=dyn_hair_04_R}
    rotation {pitch=-5}
   }
   n87 = "Transform (Modify) Bone - Bone: dyn_hair_01_R"!(ComponentPose=@n86, Alpha=@n73, Rotation="0,20,0") -> (Pose) {
    rotationMode=BMM_Additive
    rotationSpace=BCS_BoneSpace
    translationMode=BMM_Additive
    translationSpace=BCS_BoneSpace
    alphaBoolBlend {blendOption=Linear}
    boneToModify {boneName=dyn_hair_01_R}
    rotation {yaw=20}
   }
   n88 = "Transform (Modify) Bone - Bone: dyn_hair_01_R"(ComponentPose=@n87, Alpha=@n75, Rotation="0,-20,0") {
    rotationMode=BMM_Additive
    rotationSpace=BCS_BoneSpace
    translationMode=BMM_Additive
    translationSpace=BCS_BoneSpace
    alphaBoolBlend {blendOption=Linear}
    boneToModify {boneName=dyn_hair_01_R}
    rotation {yaw=-20}
   }
   Save cached pose 'is wearing backpack'(Pose=Component To Local(ComponentPose=@n88)) {cacheName=is wearing backpack}
  ]}
 EventGraph {nodes [
   Event BlueprintUpdateAnimation()
   > n1 = Sequence!(execute) -> (then_0, then_1, then_2:Exec, then_3:Exec)
   > n2 = Gravity Override State Function DEPRECATED(CharacterPartAnimInstance=Self-Reference(), GameTimeInSeconds=GetGameTimeInSeconds(), GravityOverrideParams=Get Gravity Override Params())
   > Set Gravity Override(Gravity Override=@n2)
   > Set isWearingBackpack(isWearingBackpack:bool=CheckForBackbling()) -> (Output_Get:bool)
   n3 = Get isWearingBackpack()
   > Set skydiving without backpack(skydiving without backpack:bool=AND Boolean(A=Get bIsSkyDiving(), B=NOT Boolean(A=@n3))) -> (Output_Get:bool)
   > Set SkydiveFloating(SkydiveFloating:double=SelectFloat(A=1, B=0, bPickA=Get bIsSkydiveFloating())) -> (Output_Get:double)
   > Set parachute opened(parachute opened:double=SelectFloat(A=1, B=0, bPickA=Get bIsParachuteOpened())) -> (Output_Get:double)
   > Set SkydiveDiveMode(SkydiveDiveMode:double=SelectFloat(A=1, B=0, bPickA=Get bIsSkydiveDiveMode())) -> (Output_Get:double)
   n4 = Get bIsSkyDiving()
   > Set not skydiving(not skydiving:double=SelectFloat(A=1, B=0, bPickA=NOT Boolean(A=@n4))) -> (Output_Get:double)
   > Set weapon(weapon:double=SelectFloat(A=1, B=0, bPickA=Get weapon check() -> (weapon check:bool))) -> (Output_Get:double)
   n5 = Gravity Override State Function DEPRECATED(
    execute=@n1.then_1
    CharacterPartAnimInstance=Self-Reference()
    GameTimeInSeconds=GetGameTimeInSeconds()
    GravityOverrideParams=Get Backpack Gravity Override Params() -> (Backpack Gravity Override Params:Gravity Override Params Struct Structure)
   )
   > "Set backpack Gravity Override "("backpack Gravity Override ":Vector=@n5) -> (Output_Get:Vector)
   > n6 = Cast To FortPlayerAnimInstance(Object=GetAnimInstance(self=GetOwnerSkeletalMeshComponent()))
   > Set weapon check(weapon check:bool=false) -> (Output_Get:bool)
   > n7 = Switch on EFortWeaponCoreAnimation!(execute, Selection:EFortWeaponCoreAnimation Enum=Get WeaponCoreAnim(self=@n6)) -> (
    Melee:Exec
    Pistol:Exec
    Shotgun:Exec
    PaperBlueprint:Exec
    Rifle:Exec
    MeleeOneHand:Exec
    MachinePistol:Exec
    RocketLauncher:Exec
    GrenadeLauncher:Exec
    GoingCommando:Exec
    AssaultRifle:Exec
    TacticalShotgun:Exec
    SniperRifle:Exec
    TrapPlacement:Exec
    ShoulderLauncher:Exec
    AbilityDecoTool:Exec
    Crossbow:Exec
    C4:Exec
    RemoteControl:Exec
    DualWield:Exec
    AR_BullPup:Exec
    AR_ForwardGrip:Exec
    MedPackPaddles:Exec
    SMG_P90:Exec
    AR_DrumGun:Exec
    Consumable_Small:Exec
    Consumable_Large:Exec
    Balloon:Exec
    MountedTurret:Exec
    CreativeTool:Exec
    ExplosiveBow:Exec
    AshtonIndigo:Exec
    AshtonChicago:Exec
    MeleeDualWield:Exec
    Unarmed:Exec
   )
   Set weapon check(execute=[@n7.AssaultRifle, @n7.Shotgun], weapon check=true)
   n8 = Get body mesh() -> (body mesh:Skeletal Mesh Component Budgeted Object Reference)
   Is Valid(exec=@n1.then_2, InputObject=@n8)
   n9 = BreakTransform(InTransform=GetSocketTransform(self=@n8, InSocketName=upperarm_l, TransformSpace=RTS_World))
   > Set translation(translation:Vector=@n9) -> (Output_Get:Vector)
   > Set Rotation(Rotation:Rotator=@n9.Rotation) -> (Output_Get:Rotator)
   n10 = Get body mesh()
   Set head pitch compinsation(
    execute=@n1.then_3
    head pitch compinsation:double=GetAlphaFromReferenceTransformDelta(SkeletalMeshComponent=@n10, BoneName=head, BaseName=pelvis, InputRange=(X=0,Y=10), OutputRange=(X=1,Y=0), TwistAxis="1,0,0")
   ) -> (Output_Get:double)
   n11 = GetAlphaFromReferenceTransformDelta(SkeletalMeshComponent=@n10, BoneName=head, BaseName=spine_05, InputRange=(X=-60,Y=-90), TwistAxis="1,0,0")
   n12 = "Get head yaw "() -> ("head yaw ":float)
   > Set head pitch2(head pitch2:float=float * float(A=float * float(A=@n11, B=@n12), B=Get head pitch compinsation() -> (head pitch compinsation:double)) {is=title K2Node_PromotableOperator}) -> (Output_Get:float)
   n13 = GetAlphaFromReferenceTransformDelta(SkeletalMeshComponent=@n10, BoneName=head, BaseName=spine_05, InputRange=(X=-20,Y=-60), TwistAxis="1,0,0")
   n14 = "Get head yaw "()
   > Set head pitch1(head pitch1:double=float * float(A=float * float(A=@n13, B=@n14), B=Get head pitch compinsation())) -> (Output_Get:double)
   n15 = GetAlphaFromReferenceTransformDelta(SkeletalMeshComponent=@n10, BoneName=head, BaseName=spine_05, DeltaComponentType=EulerY, InputRange=(X=0,Y=50), TwistAxis="1,0,0")
   n16 = "Get head yaw "()
   > Set head roll(head roll:float=float * float(A=float * float(A=@n15, B=@n16), B=Get head pitch compinsation())) -> (Output_Get:float)
   n17 = SelectFloat(A=-10, B=0, bPickA=OR Boolean(A=Get bIsPlayingEmote(), B=Get bIsPlayingMeleeAnim()))
   n18 = GetAlphaFromReferenceTransformDelta(
    SkeletalMeshComponent=@n10
    BoneName=head
    BaseName=spine_05
    DeltaComponentType=EulerX
    InputRange=(X=0,Y=0)
    OutputRange=(X=0,Y=0)
    TwistAxis="1,0,0"
   )
   > "Set head yaw "("head yaw ":float=float * float(
     A=MapRangeClamped(Value=@n18.RawOutput, InRangeA=30, InRangeB=50, OutRangeA=1, OutRangeB=@n17)
     B=MapRangeClamped(Value=@n18.RawOutput, InRangeA=-30, InRangeB=-50, OutRangeA=1, OutRangeB=@n17)
    )) -> (Output_Get:float)
   > Set head yaw left skydiving(head yaw left skydiving:float=SelectFloat(A=MapRangeClamped(Value=@n18.RawOutput, InRangeA=0, InRangeB=25), B=0, bPickA=Get bIsSkydiveFloating())) -> (Output_Get:float)
   > Set head yaw right skydiving(head yaw right skydiving:double=SelectFloat(A=MapRangeClamped(Value=@n18.RawOutput, InRangeA=0, InRangeB=-25), B=0, bPickA=Get bIsSkydiveFloating())) -> (Output_Get:double)
   n19 = Event OnCharacterCustomizationLoaded()
   > Set body mesh(body mesh:Skeletal Mesh Component Budgeted Object Reference=Get Array Item with EFort Custom Part Type(Array=@n19.CustomMeshParts)) -> (Output_Get:Skeletal Mesh Component Budgeted Object Reference)
   n20 = SelectVector(A="0.1,0.1,0.1", B="0.5,0.5,0.5", bPickA=Get bIsSkyDiving())
   Set Component Linear Vel Scale(Component Linear Vel Scale:Vector=@n20) -> (Output_Get:Vector)
   > Set Component Linear Acc Scale(Component Linear Acc Scale:Vector=@n20) -> (Output_Get:Vector)
   PrintString!(execute, InString="To String (Float)"(InDouble=Get SkydiveFloating()), bPrintToScreen, bPrintToLog, TextColor, Duration, Key) -> (then)
   Get SkydiveDiveMode()
  ]}
]
variables [
 backpack {type=SkeletalMeshComponentBudgeted}
 "backpack Gravity Override " {type=Vector}
 Backpack Gravity Override Params {
  type=GravityOverrideParamsStruct
  default {skydive_FloatingDown {windAmplitude=<200 -100 100>}, skydive_Parachuting {windAmplitude=<200 0 200>}}
 }
 body mesh {type=SkeletalMeshComponentBudgeted}
 Component Linear Acc Scale {type=Vector, default=<0.5 0.5 0.5>}
 Component Linear Vel Scale {type=Vector, default=<0.5 0.5 0.5>}
 Gravity Override {type=Vector}
 Gravity Override Params {
  type=GravityOverrideParamsStruct
  default {
   skydive_FloatingDown {windAmplitude=<200 -200 200>, windFrequency=<1.5 7 10>}
   skydive_GoingUp {windFrequency=<1.5 5 7>}
   skydive_Parachuting {windAmplitude=<200 0 200>, windFrequency=<1.5 5 8>}
  }
 }
 head pitch compinsation {type=number}
 head pitch1 {type=number}
 head pitch2 {type=number}
 head roll {type=number}
 "head yaw " {type=number}
 head yaw left skydiving {type=number}
 head yaw right skydiving {type=number}
 isWearingBackpack {type=boolean}
 not skydiving {type=number}
 parachute opened {type=number}
 Rotation {type=Rotator}
 SkydiveDiveMode {type=number}
 SkydiveFloating {type=number}
 skydiving without backpack {type=boolean}
 translation {type=Vector}
 weapon {type=number}
 weapon check {type=boolean}
]
